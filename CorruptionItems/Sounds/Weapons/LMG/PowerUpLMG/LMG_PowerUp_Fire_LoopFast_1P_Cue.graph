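# First-person sound cue for the fast firing loop of the power-up LMG.
# A mixer layers a looping fire wave pitched slightly down, a looping barrel wave and a pitched-up cube wave that is not set to loop.
# It uses first-person weapon fire class, attenuation and gunfire concurrency assets, with manual reverb send and owner-limited, stop-oldest concurrency overrides also defined.
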SoundCue /CorruptionItems/Sounds/Weapons/LMG/PowerUpLMG/LMG_PowerUp_Fire_LoopFast_1P_Cue
root=Mixer_0
nodes [
 {id=Mixer_0, children [Modulator_0,WavePlayer_1,Modulator_1], properties {inputVolume [0.8,0.6,0.9]}}
 {id=Modulator_0, children [WavePlayer_0], properties {pitchMax=0.95}}
 {id=Modulator_1, children [WavePlayer_3], properties {pitchMax=1.2, pitchMin=1.2}}
 {id=WavePlayer_0, properties {bLooping=true, soundWaveAssetPtr=&/LMG_Powerup_Fire_LoopFast_01}}
 {id=WavePlayer_1, properties {bLooping=true, soundWaveAssetPtr=&/LMG_PowerUp_Barrel_FastLoop_01}}
 {id=WavePlayer_3, properties {soundWaveAssetPtr=&/LMG_Powerup_Cube_Loop_01}}
]
properties {
 attenuationSettings=&Fort_Weapon_Projectile_FirstPerson_Athena_Att
 maxDistance=2097152
 soundClassObject=&Fort_WeaponRanged_Fire_1P_PSM
 volumeMultiplier=0.5
 attenuationOverrides {manualReverbSendLevel=-1.0165, reverbSendMethod=Manual}
 concurrencyOverrides {bLimitToOwner=true, maxCount=3, resolutionRule=StopOldest, volumeScale=0.9}
 concurrencySet [&Weapon_Gunfire_FirstPerson_Conc]
}
