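# First-person looping sound cue for the slow fire loop of the power-up LMG weapon.
# A mixer layers three looping waves at different levels: a slow fire loop, a barrel loop and a cube loop.
# Modulators hold the barrel and cube layers at fixed lowered pitches, with no random variation.
# The cue routes through a first-person ranged-fire sound class, projectile attenuation and first-person gunfire concurrency.
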SoundCue /CorruptionItems/Sounds/Weapons/LMG/PowerUpLMG/LMG_PowerUp_Fire_LoopSlow_1P_Cue
root=Mixer_2
nodes [
 {id=Mixer_2, children [WavePlayer_2,Modulator_0,Modulator_1], properties {inputVolume [0.7,0.5,1]}}
 {id=Modulator_0, children [WavePlayer_6], properties {pitchMax=0.8, pitchMin=0.8}}
 {id=Modulator_1, children [WavePlayer_0], properties {pitchMax=0.9, pitchMin=0.9}}
 {id=WavePlayer_0, properties {bLooping=true, soundWaveAssetPtr=&/LMG_Powerup_Cube_Loop_01}}
 {id=WavePlayer_2, properties {bLooping=true, soundWaveAssetPtr=&/LMG_Powerup_Fire_LoopSlow_01}}
 {id=WavePlayer_6, properties {bLooping=true, soundWaveAssetPtr=&/LMG_PowerUp_Barrel_SlowLoop_01}}
]
properties {
 attenuationSettings=&Fort_Weapon_Projectile_FirstPerson_Athena_Att
 bDebug=true
 maxDistance=2097152
 soundClassObject=&Fort_WeaponRanged_Fire_1P_PSM
 volumeMultiplier=0.35
 attenuationOverrides {manualReverbSendLevel=-1.0165, reverbSendMethod=Manual}
 concurrencyOverrides {bLimitToOwner=true, maxCount=3, resolutionRule=StopOldest, volumeScale=0.9}
 concurrencySet [&Weapon_Gunfire_FirstPerson_Conc]
}
