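# First-person Sound Cue for the stop sound of the power-up LMG's slow firing loop.
# A single Wave Player feeds a Modulator root that applies a fixed upward pitch shift, with no random variation.
# The cue plays at reduced volume through first-person weapon attenuation, the ranged-fire 1P sound class and first-person gunfire concurrency, with its debug flag enabled.
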SoundCue /CorruptionItems/Sounds/Weapons/LMG/PowerUpLMG/LMG_PowerUp_Fire_LoopSlow_Stop_1P_Cue
root=Modulator_1
nodes [
 {id=Modulator_1, children [WavePlayer_3], properties {pitchMax=1.2, pitchMin=1.2}}
 {id=WavePlayer_3, properties {soundWaveAssetPtr=&/LMG_PowerUp_Fire_LoopSlow_Stop_01}}
]
properties {
 attenuationSettings=&Fort_Weapon_Projectile_FirstPerson_Athena_Att
 bDebug=true
 maxDistance=2097152
 soundClassObject=&Fort_WeaponRanged_Fire_1P_PSM
 volumeMultiplier=0.25
 concurrencySet [&Weapon_Gunfire_FirstPerson_Conc]
}
