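# Sound cue for the powered-up LMG's bullet impacts on flesh.
# A mixer layers a random pick of four LMG power-up impact waves, pitched down to a fixed rate by a modulator, with a random pick of three reused Galileo Bun body-impact waves.
# It uses bullet-impact attenuation and the impacts sound class, with a concurrency override that caps simultaneous voices and blocks new ones past the limit.
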SoundCue /CorruptionItems/Sounds/Weapons/LMG/PowerUpLMG/LMG_PowerUp_ImpactFlesh_Cue
root=Mixer_0
nodes [
 {id=Mixer_0, children [Modulator_0,Random_6,null], properties {inputVolume [1,1,0]}}
 {id=Modulator_0, children [Random_3], properties {pitchMax=0.8, pitchMin=0.8}}
 {id=Random_3, children [WavePlayer_0,WavePlayer_1,WavePlayer_2,WavePlayer_3], properties {weights [1,1,1,1]}}
 {id=Random_6, children [WavePlayer_18, WavePlayer_19, WavePlayer_20], properties {weights [1,1,1]}}
 {id=WavePlayer_0, properties {soundWaveAssetPtr=&/LMG_PowerUp_Impact_01}}
 {id=WavePlayer_1, properties {soundWaveAssetPtr=&/LMG_PowerUp_Impact_02}}
 {id=WavePlayer_18, properties {soundWaveAssetPtr=/Game/Sounds/Fort_Weapons/AutoRifle/Galileo_Bun/Galileo_Bun_Impact_Body_02}}
 {id=WavePlayer_19, properties {soundWaveAssetPtr=/Game/Sounds/Fort_Weapons/AutoRifle/Galileo_Bun/Galileo_Bun_Impact_Body_04}}
 {id=WavePlayer_2, properties {soundWaveAssetPtr=&/LMG_PowerUp_Impact_03}}
 {id=WavePlayer_20, properties {soundWaveAssetPtr=/Game/Sounds/Fort_Weapons/AutoRifle/Galileo_Bun/Galileo_Bun_Impact_Body_03}}
 {id=WavePlayer_3, properties {soundWaveAssetPtr=&/LMG_PowerUp_Impact_04}}
]
properties {
 attenuationSettings=/Game/Sounds/Attenuation/Fort_Bullet_Impact_Athena_Att
 bOverrideConcurrency=true
 maxDistance=4500
 soundClassObject=&Fort_Impacts
 volumeMultiplier=0.9
 concurrencyOverrides {maxCount=4, resolutionRule=PreventNew}
}
