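# Sound cue for the landing or drop sound of the LMG power-up item.
# A Random node picks one of three drop wave variations with equal weight.
# It uses the loot collect sound class with overridden short-range attenuation and a concurrency limit that blocks new instances once a few are already playing.
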SoundCue /CorruptionItems/Sounds/Weapons/LMG/PowerUpLMG/LMG_PowerUp_Landed_Cue
root=Random_2
nodes [
 {id=Random_2, children [WavePlayer_0,WavePlayer_1,WavePlayer_2], properties {weights [1,1,1]}}
 {id=WavePlayer_0, properties {soundWaveAssetPtr=&/LMG_PowerUp_Drop_01}}
 {id=WavePlayer_1, properties {soundWaveAssetPtr=&/LMG_PowerUp_Drop_02}}
 {id=WavePlayer_2, properties {soundWaveAssetPtr=&/LMG_PowerUp_Drop_03}}
]
properties {
 bOverrideAttenuation=true
 bOverrideConcurrency=true
 maxDistance=1000
 soundClassObject=&Fort_Loot_Collect
 volumeMultiplier=1
 attenuationOverrides {falloffDistance=1000, attenuationShapeExtents {x=0}}
 concurrencyOverrides {maxCount=3, resolutionRule=PreventNew}
}
