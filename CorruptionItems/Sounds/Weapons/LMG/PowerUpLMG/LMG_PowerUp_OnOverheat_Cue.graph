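# Sound cue for the overheat of the power-up LMG in the CorruptionItems weapon set.
# A mixer layers a stop sound and an overheat wind-down with one of three equally weighted cube overheat variations chosen at random.
# It routes through a third-person ranged weapon fire sound class with close-projectile attenuation settings.
# Its concurrency override prevents new instances once the limit is reached.
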SoundCue /CorruptionItems/Sounds/Weapons/LMG/PowerUpLMG/LMG_PowerUp_OnOverheat_Cue
root=Mixer_0
nodes [
 {id=Mixer_0, children [WavePlayer_9, WavePlayer_12, Random_3], properties {inputVolume [0.8,1,0.8]}}
 {id=Random_3, children [WavePlayer_17, WavePlayer_18, WavePlayer_19], properties {weights [1,1,1]}}
 {id=WavePlayer_12, properties {soundWaveAssetPtr=&/LMG_Powerup_OverheatWindDown_01}}
 {id=WavePlayer_17, properties {soundWaveAssetPtr=&/LMG_Powerup_Cube_Overheat_01}}
 {id=WavePlayer_18, properties {soundWaveAssetPtr=&/LMG_Powerup_Cube_Overheat_02}}
 {id=WavePlayer_19, properties {soundWaveAssetPtr=&/LMG_Powerup_Cube_Overheat_03}}
 {id=WavePlayer_9, properties {soundWaveAssetPtr=&/LMG_PowerUp_KevinStop_06}}
]
properties {
 attenuationSettings=&Fort_Weapon_Projectile_Close_Athena_Att
 bOverrideConcurrency=true
 maxDistance=4950
 soundClassObject=&Fort_WeaponRanged_Fire_3P_PSM
 volumeMultiplier=0.27
 attenuationOverrides {distanceAlgorithm=NaturalSound, falloffDistance=4000, attenuationShapeExtents {x=1000}}
 concurrencyOverrides {resolutionRule=PreventNew}
}
