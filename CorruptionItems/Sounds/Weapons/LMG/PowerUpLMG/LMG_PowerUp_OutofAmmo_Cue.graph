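# Sound cue for the out-of-ammo sound of the power-up LMG weapon.
# A mixer layers the out-of-ammo wave, with slight random volume variation, and a random pick of two KevinStop waves, both inputs at reduced volume.
# It uses the third-person ranged weapon fire sound class with its own falloff distance.
# A concurrency override caps simultaneous instances at two, stopping the oldest.
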SoundCue /CorruptionItems/Sounds/Weapons/LMG/PowerUpLMG/LMG_PowerUp_OutofAmmo_Cue
root=Mixer_0
nodes [
 {id=Mixer_0, children [Modulator_0,Random_1], properties {inputVolume [0.5,0.5]}}
 {id=Modulator_0, children [WavePlayer_4], properties {pitchMax=1, pitchMin=1, volumeMax=0.95, volumeMin=0.85}}
 {id=Random_1, children [WavePlayer_6,WavePlayer_8], properties {weights [1,1]}}
 {id=WavePlayer_4, properties {soundWaveAssetPtr=&/LMG_PowerUp_OutofAmmo_01}}
 {id=WavePlayer_6, properties {soundWaveAssetPtr=&/LMG_PowerUp_KevinStop_02}}
 {id=WavePlayer_8, properties {soundWaveAssetPtr=&/LMG_PowerUp_KevinStop_04}}
]
properties {
 bOverrideConcurrency=true
 maxDistance=2097152
 soundClassObject=&Fort_WeaponRanged_Fire_3P_PSM
 volumeMultiplier=0.25
 attenuationOverrides {falloffDistance=1600}
 concurrencyOverrides {maxCount=2, resolutionRule=StopOldest}
}
