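# Sound cue for the first-person overheat loop of the power-up LMG.
# A looping wave player feeds a continuous modulator that scales pitch and volume from the Overheat parameter, then a single-input mixer at the root.
# It references first-person projectile attenuation and the global and local gunfire concurrency groups.
# Despite its 1P name, it is assigned the third-person ranged-fire sound class.
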SoundCue /CorruptionItems/Sounds/Weapons/LMG/PowerUpLMG/LMG_PowerUp_OverheatLoop_1P_Cue
root=Mixer_0
nodes [
 {id=Mixer_0, children [ModulatorContinuous_1], properties {inputVolume [1]}}
 {
  id=ModulatorContinuous_1
  children [WavePlayer_1]
  properties {
   pitchModulationParams {default=0, maxOutput=1.4, minOutput=0.85, parameterName=Overheat}
   volumeModulationParams {default=0, maxOutput=0.9, parameterName=Overheat}
  }
 }
 {id=WavePlayer_1, properties {bLooping=true, soundWaveAssetPtr=&/LMG_PowerUp_Power_Loop_02}}
]
properties {
 attenuationSettings=&Fort_Weapon_Projectile_FirstPerson_Athena_Att
 maxDistance=2097152
 soundClassObject=&Fort_WeaponRanged_Fire_3P_PSM
 volumeMultiplier=0.35
 attenuationOverrides {distanceAlgorithm=NaturalSound, falloffDistance=4000, attenuationShapeExtents {x=1000}}
 concurrencyOverrides {resolutionRule=PreventNew}
 concurrencySet [&Weapon_Gunfire_Global_Conc, &Weapon_Gunfire_Local_Conc]
}
