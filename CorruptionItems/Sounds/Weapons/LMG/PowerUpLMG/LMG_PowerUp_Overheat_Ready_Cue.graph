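# Sound cue for the LMG power-up overheat-ready sound, routed through the weapon foley class with reload attenuation.
# A mixer layers a power-up start wave, a kick thump and one of four equally weighted cube overheat variations held back by a short fixed delay.
# Concurrency allows a single instance per owner and blocks new plays while one is active.
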SoundCue /CorruptionItems/Sounds/Weapons/LMG/PowerUpLMG/LMG_PowerUp_Overheat_Ready_Cue
root=Mixer_0
nodes [
 {id=Delay_1, children [Random_1], properties {delayMax=0.3, delayMin=0.3}}
 {id=Mixer_0, children [WavePlayer_0,Delay_1,WavePlayer_8], properties {inputVolume [0.7,1,0.6]}}
 {id=Random_1, children [WavePlayer_9, WavePlayer_10, WavePlayer_11, WavePlayer_12], properties {weights [1,1,1,1]}}
 {id=WavePlayer_0, properties {soundWaveAssetPtr=&/LMG_Powerup_Fire_Powerup_Start_01}}
 {id=WavePlayer_10, properties {soundWaveAssetPtr=&/LMG_Powerup_Cube_Overheat_02}}
 {id=WavePlayer_11, properties {soundWaveAssetPtr=&/LMG_Powerup_Cube_Overheat_03}}
 {id=WavePlayer_12, properties {soundWaveAssetPtr=&/LMG_Powerup_Cube_Overheat_04}}
 {id=WavePlayer_8, properties {soundWaveAssetPtr=&/LMG_Kick_Thump_01}}
 {id=WavePlayer_9, properties {soundWaveAssetPtr=&/LMG_Powerup_Cube_Overheat_01}}
]
properties {
 attenuationSettings=&Fort_Weapon_Reload_Att
 bOverrideConcurrency=true
 maxDistance=2250
 soundClassObject=&Fort_WeaponFoley
 volumeMultiplier=0.401
 concurrencyOverrides {bLimitToOwner=true, maxCount=1, resolutionRule=PreventNew}
}
