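# Sound cue for picking up the LMG power-up weapon item, classed as a loot-collect sound.
# A mixer layers a random choice of two pickup waves with a cube pickup wave, each branch passing through its own modulator for slight pitch variation.
# Playback uses weapon-pickup attenuation and a shared pickup concurrency set, with overall pitch raised and volume reduced.
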SoundCue /CorruptionItems/Sounds/Weapons/LMG/PowerUpLMG/LMG_PowerUp_Pickup_Cue
root=Mixer_0
nodes [
 {id=Mixer_0, children [Modulator_0,Modulator_1], properties {inputVolume [1,1]}}
 {id=Modulator_0, children [Random_1], properties {pitchMax=1.1, pitchMin=0.9}}
 {id=Modulator_1, children [WavePlayer_0], properties {pitchMax=1.1, pitchMin=0.9}}
 {id=Random_1, children [WavePlayer_7,WavePlayer_8], properties {weights [1,1]}}
 {id=WavePlayer_0, properties {soundWaveAssetPtr=&/LMG_Powerup_Cube_Pickup_01}}
 {id=WavePlayer_7, properties {soundWaveAssetPtr=&/LMG_PowerUp_Pickup_01}}
 {id=WavePlayer_8, properties {soundWaveAssetPtr=&/LMG_PowerUp_Pickup_02}}
]
properties {
 attenuationSettings=&Fort_Weapon_Pickup_Att
 maxDistance=3500
 pitchMultiplier=1.1
 soundClassObject=&Fort_Loot_Collect
 volumeMultiplier=0.5538
 concurrencyOverrides {maxCount=1, resolutionRule=PreventNew}
 concurrencySet [&PickupMaximum_Oldest]
}
