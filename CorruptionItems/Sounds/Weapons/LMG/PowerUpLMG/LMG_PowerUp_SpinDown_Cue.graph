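# Weapon spin-down sound cue for the power-up LMG, currently silenced by a zero volume multiplier.
# A mixer layers a close-attenuated random pick of two barrel slow-stop waves with a half-volume distant GoinCommando minigun spin-down.
# The distant layer switches between standard and optimized waves by quality level.
# Both branches pass through modulators and override their sound class to third-person ranged weapon fire.
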SoundCue /CorruptionItems/Sounds/Weapons/LMG/PowerUpLMG/LMG_PowerUp_SpinDown_Cue
root=Mixer_17
nodes [
 {
  id=Attenuation_0
  children [Modulator_1]
  properties {attenuationSettings=/Game/Sounds/Attenuation/Fort_Weapon_Projectile_Medium_MinigunSpin_Athena_Att}
 }
 {id=Attenuation_49, children [Modulator_0], properties {attenuationSettings=&Fort_Weapon_Projectile_Close_Athena_Att}}
 {id=Mixer_17, children [SoundClass_32, SoundClass_0], properties {inputVolume [1,0.5]}}
 {id=Modulator_0, children [Random_0]}
 {id=Modulator_1, children [QualityLevel_0]}
 {id=QualityLevel_0, children [WavePlayer_2,WavePlayer_2,WavePlayer_0,WavePlayer_0]}
 {id=Random_0, children [WavePlayer_3,WavePlayer_4], properties {weights [1,1]}}
 {
  id=SoundClass_0
  children [Attenuation_0]
  properties {soundClassOverride=&Fort_WeaponRanged_Fire_3P_PSM}
 }
 {
  id=SoundClass_32
  children [Attenuation_49]
  properties {soundClassOverride=&Fort_WeaponRanged_Fire_3P_PSM}
 }
 {id=WavePlayer_0, properties {soundWaveAssetPtr=/Game/Sounds/Fort_Weapons/AutoRifle/WPN_GoinCommando_SpinDown_distant_athena_opt}}
 {id=WavePlayer_2, properties {soundWaveAssetPtr=/Game/Sounds/Fort_Weapons/AutoRifle/WPN_GoinCommando_SpinDown_distant_athena}}
 {id=WavePlayer_3, properties {soundWaveAssetPtr=&/LMG_PowerUp_Barrel_SlowStop_01}}
 {id=WavePlayer_4, properties {soundWaveAssetPtr=&/LMG_PowerUp_Barrel_SlowStop_02}}
]
properties {
 bDebug=true
 bOverrideConcurrency=true
 maxDistance=4950
 soundClassObject=&Fort_WeaponRanged
 volumeMultiplier=0
}
