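# Sound cue for the power-up LMG's barrel slow-start spin-up in the CorruptionItems content.
# It mixes a close layer with a quieter distant layer, each picking randomly from three wave variants through a modulator.
# Each layer has its own attenuation and sound class override, with the distant layer routed to the third-person ranged fire class.
# The cue overrides concurrency, lowers overall pitch and volume, and has its debug flag on.
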SoundCue /CorruptionItems/Sounds/Weapons/LMG/PowerUpLMG/LMG_PowerUp_SpinUp_Cue
root=Mixer_17
nodes [
 {
  id=Attenuation_0
  children [Modulator_1]
  properties {attenuationSettings=&Fort_Weapon_Projectile_Medium_Mono_Athena_Att}
 }
 {id=Attenuation_49, children [Modulator_0], properties {attenuationSettings=&Fort_Weapon_Projectile_Close_Athena_Att}}
 {id=Mixer_17, children [SoundClass_32, SoundClass_0], properties {inputVolume [1,0.7]}}
 {id=Modulator_0, children [Random_0]}
 {id=Modulator_1, children [Random_1]}
 {id=Random_0, children [WavePlayer_1,WavePlayer_2,WavePlayer_3], properties {weights [1,1,1]}}
 {id=Random_1, children [WavePlayer_4,WavePlayer_5,WavePlayer_6], properties {weights [1,1,1]}}
 {
  id=SoundClass_0
  children [Attenuation_0]
  properties {soundClassOverride=&Fort_WeaponRanged_Fire_3P_PSM}
 }
 {id=SoundClass_32, children [Attenuation_49], properties {soundClassOverride=/Game/Sounds/FortSoundClassesAndModes/Fort_Weapons}}
 {id=WavePlayer_1, properties {soundWaveAssetPtr=&/LMG_PowerUp_Barrel_SlowStart_02}}
 {id=WavePlayer_2, properties {soundWaveAssetPtr=&/LMG_PowerUp_Barrel_SlowStart_01}}
 {id=WavePlayer_3, properties {soundWaveAssetPtr=&/LMG_PowerUp_Barrel_SlowStart_03}}
 {id=WavePlayer_4, properties {soundWaveAssetPtr=&/LMG_PowerUp_Barrel_SlowStart_Distant03}}
 {id=WavePlayer_5, properties {soundWaveAssetPtr=&/LMG_PowerUp_Barrel_SlowStart_Distant_01}}
 {id=WavePlayer_6, properties {soundWaveAssetPtr=&/LMG_PowerUp_Barrel_SlowStart_Distant_02}}
]
properties {
 bDebug=true
 bOverrideConcurrency=true
 maxDistance=10579.2871
 pitchMultiplier=0.8
 soundClassObject=&Fort_WeaponRanged
 volumeMultiplier=0.4
 concurrencyOverrides {voiceStealReleaseTime=0.4}
}
